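; Parent actor for an environmental plant pod that launches or explodes when hit by weapons, sliding players or vehicles.
; Either reaction spawns a projectile wearing the pod's mesh and carrying gameplay effect specs, with explosions blowing it up.
; A box trace from the instigator can steer the projectile toward a team-filtered moving target.
; A replicated flag hides the pod until a timer regrows it, retrying while the space is occupied.
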(event EventBeginPlay
  (bind _fortsoundindicator (Variables|Default|GetFortSoundIndicator))
  ;[0 0.8244 1] Begin Play
  (Audio|Components|Audio|SetSound (Variables|Default|GetAmbient%20Loop) (Variables|Default|GetAmbient%20Sound))
  (CallFunction|FadeinAmbientLoop)
  (Components|Activation|Deactivate _fortsoundindicator)
  (SoundIndicator|StopSoundIndicatorLooping _fortsoundindicator))

(event EventOnDamageServer
  (params (Damage :type Float%20%28single-precision%29) (DamageTags :type Gameplay%20Tag%20Container%20Structure%20%28by%20ref%29) (Momentum :type Vector :default "0, 0, 0") (HitInfo :type Hit%20Result%20Structure%20%28by%20ref%29) (InstigatedBy :type Controller%20Object%20Reference) (DamageCauser :type Actor%20Object%20Reference) (EffectContext :type Gameplay%20Effect%20Context%20Handle%20Structure))
  (bind _instigatingactor (Variables|Default|GetInstigatingActor))
  (bind _returnvalue (Transformation|GetActorForwardVector _instigatingactor))
  ;[0.4107 0.6115 0.7899] On damaged authority
  (CallFunction|ResetHealth)
  (if (not (Variables|Default|GetInteractionDisabled?))
    (bind _asfort_weapon (Utilities|Casting|CastToFortWeapon DamageCauser)
      (:then
        (Variables|Default|SetInstigatingActor (Actor|GetOwner _asfort_weapon))
        (Variables|Default|SetHitNormal _returnvalue)
        (if (ScalableFloat|GetValueasBool (Variables|Default|GetHF_AlwaysLaunch?))
          _
          (elif (Utilities|Array|ContainsItem (Variables|Default|GetItemTypesToLaunch) (Item|GetWeaponDataItemType _asfort_weapon))
            _
            (else
              (CallFunction|ExecuteExplosion false)
              (Rendering|Material|SetScalarParameterValueonMaterials (Variables|Default|GetLaunchableMesh) "WPOPuslingScale" 3.0))))
        (CallFunction|ExecuteLaunch)
        (GameplayEffects|ApplyGameplayEffectToTarget :self (Ability|GetFortAbilitySystemComponent self) :GameplayEffectClass (Variables|Default|GetGE_GrantedTag) :Target (Ability|GetFortAbilitySystemComponent (Variables|Default|GetInstigatingActor)) :Level 0.0))
      (:CastFailed
        (Default|SetInstigatingActorwithCheck DamageCauser)
        (Variables|Default|SetHitNormal _returnvalue)))))

(event Custom|ExecuteLaunch
  ;[0 0.8244 1] Do effects based on damage source
  (Collision|SetCollisionEnabled (Variables|Default|GetLaunchableMesh) "NoCollision")
  (bind _returnvalue (CallFunction|MakeProjectile))
  (Variables|Default|SetbWaitingToRegrow true)
  (Networking|SwitchHasAuthority
    (:Authority
      (CallFunction|BindtoPlantExplosion _returnvalue))
    (:Remote
      _))
  (Variables|Default|SetInteractionDisabled? true)
  (bind _returnvalue_2 (Utilities|Time|SetTimerbyEvent (EventDispatchers|CreateEvent :!event Custom|TryRegrowth) (ScalableFloat|GetValueatLevel (Variables|Default|GetHF_RegrowthDelay))))
  (Variables|Default|SetRegrowthCheckTimerHandle _returnvalue_2)
  (Ability|GameplayEffect|AddLooseGameplayTags self (Variables|Default|GetRegrowthTags))
  (CallFunction|FadeOutAmbientLoop))

(event Custom|ExecuteExplosion
  (params (bExplodeInstantly :type Boolean))
  ;[0 0.8244 1] Do effects based on damage source
  (Collision|SetCollisionEnabled (Variables|Default|GetLaunchableMesh) "NoCollision")
  (bind _returnvalue (CallFunction|MakeProjectile))
  (Variables|Default|SetbWaitingToRegrow true)
  (Class|PrjChronoLaunchablePlantParent|BlowUp _returnvalue bExplodeInstantly 0)
  (CallFunction|BindtoPlantExplosion _returnvalue)
  (Variables|Default|SetInteractionDisabled? true)
  (bind _returnvalue_2 (Utilities|Time|SetTimerbyEvent (EventDispatchers|CreateEvent :!event Custom|TryRegrowth) (ScalableFloat|GetValueatLevel (Variables|Default|GetHF_RegrowthDelay))))
  (Variables|Default|SetRegrowthCheckTimerHandle _returnvalue_2)
  (Ability|GameplayEffect|AddLooseGameplayTags self (Variables|Default|GetRegrowthTags))
  (CallFunction|FadeOutAmbientLoop))

(event Custom|TryRegrowth
  ;[0.03429 0.3069 0.7899] Regrowth functionality
  (Utilities|CanRegrow?
    (:True
      (Variables|Default|SetbWaitingToRegrow false)
      (Variables|Default|SetInteractionDisabled? false)
      (Collision|SetCollisionEnabled (Variables|Default|GetLaunchableMesh) "QueryAndPhysics")
      (Ability|GameplayEffect|RemoveLooseGameplayTags self (Variables|Default|GetRegrowthTags)))))

(event OnComponentHit(LaunchableMesh)
  (params (HitComponent :type Primitive%20Component%20Object%20Reference) (OtherActor :type Actor%20Object%20Reference) (OtherComp :type Primitive%20Component%20Object%20Reference) (NormalImpulse :type Vector :default "0, 0, 0") (Hit :type Hit%20Result%20Structure%20%28by%20ref%29))
  ;[0 0.8244 1] Slide kick functionality
  (if (not (Variables|Default|GetInteractionDisabled?))
    (Utilities|DoOncePerFrame
      (:%20%20
        (bind _returnvalue (GameplayTags|HasMatchingGameplayTag(Message) OtherActor "(TagName=\"MovementMode.Sliding\")"))
        (if _returnvalue
          (Variables|Default|SetInstigatingActor OtherActor)
          (Variables|Default|SetHitNormal (Transformation|GetActorForwardVector OtherActor))
          (if (Variables|Default|GetbExplodeOnSlideKick)
            (CallFunction|ExecuteExplosion true)
            (else
              (CallFunction|ExecuteLaunch)))
          (CallFunction|FadeOutAmbientLoop))))))

(event OnComponentBeginOverlap(LaunchableMesh)
  (params (OverlappedComponent :type Primitive%20Component%20Object%20Reference) (OtherActor :type Actor%20Object%20Reference) (OtherComp :type Primitive%20Component%20Object%20Reference) (OtherBodyIndex :type Integer) (bFromSweep :type Boolean) (SweepResult :type Hit%20Result%20Structure%20%28by%20ref%29))
  ;[0.9394 0.2521 1] Vehicle interaction
  (if (not (Variables|Default|GetInteractionDisabled?))
    (bind _asfort_vehicle_interface (Utilities|Casting|CastToFortVehicleInterface OtherActor))
    (Variables|Default|SetInstigatingActor (Vehicle|GetDriver _asfort_vehicle_interface))
    (Variables|Default|SetHitNormal (Transformation|GetActorForwardVector OtherActor))
    (if (Variables|Default|GetbExplodeOnSlideKick)
      (CallFunction|ExecuteExplosion true)
      (else
        (CallFunction|ExecuteLaunch)))
    (CallFunction|FadeOutAmbientLoop)))

(event EventOnDeathPlayEffects
  (params (Damage :type Float%20%28single-precision%29) (DamageTags :type Gameplay%20Tag%20Container%20Structure%20%28by%20ref%29) (Momentum :type Vector :default "0, 0, 0") (HitInfo :type Hit%20Result%20Structure%20%28by%20ref%29) (InstigatedBy :type Fort%20Pawn%20Object%20Reference) (DamageCauser :type Actor%20Object%20Reference) (EffectContext :type Gameplay%20Effect%20Context%20Handle%20Structure))
  ;[0.03429 0.3069 0.7899] OnDeath effects
  (CallFunction|FadeOutAmbientLoop))

(fn BindToPlantExplosion
  (params (Projectile :type Prj%20Chrono%20Launchable%20Plant%20Parent%20Object%20Reference))
  ;[0.4546 0.06953 0.6667] Ensure that we always run OnProjectileHit, even when the pod gets destroyed instantly
  (Utilities|IsValid Projectile
    (:Is%20Valid
      (Default|BindEventtoOnPlantExploded Projectile (EventDispatchers|CreateEvent :!event OnProjectileHit)))
    (:Is%20Not%20Valid
      (CallFunction|OnProjectileHit))))

(fn DoTraceForInfluence
  (bind _outhits (Collision|MultiBoxTraceForObjects (Transformation|GetActorLocation (Variables|Default|GetInstigatingActor)) (+ (Transformation|GetActorLocation (Variables|Default|GetInstigatingActor)) (* (Transformation|GetActorForwardVector (Variables|Default|GetInstigatingActor)) (Variables|Default|GetTraceForwardDistance))) (Math|Vector|MakeVector 1024.0 (Variables|Default|GetTraceWidthDistance) (Variables|Default|GetTraceHeightDistance)) (Transformation|GetActorRotation (Variables|Default|GetInstigatingActor)) (Utilities|Array|MakeArray "ObjectTypeQuery3") false (Utilities|Array|MakeArray (Variables|Default|GetInstigatingActor)) "ForDuration"))
  (switch Utilities|FlowControl|Switch|SwitchonEPodPlantTargetType (Variables|Default|GetTargetingType)
    (:NewEnumerator2
      _)
    (:NewEnumerator0
      (bind _array_element (Utilities|Array|ForEachLoopwithBreak _outhits)
        (:LoopBody
          (if (and (Targeting|IsPlantBetweenInstigatorandTarget (Variables|Default|GetInstigatingActor) (.HitActor (Collision|BreakHitResult _array_element))) (Utilities|Select :Index (Variables|Default|GetTargetingType) :NewEnumerator2 false :NewEnumerator0 (Utilities|Select :Index (Team|GetTeamAffiliation (.HitActor (Collision|BreakHitResult _array_element)) (Variables|Default|GetInstigatingActor)) :Friendly false :Neutral false :Hostile true) :NewEnumerator1 (Utilities|Select :Index (Team|GetTeamAffiliation (.HitActor (Collision|BreakHitResult _array_element)) (Variables|Default|GetInstigatingActor)) :Friendly true :Neutral false :Hostile false)))
            (LocalVariables|SetTargetActorResult (.HitActor (Collision|BreakHitResult _array_element)))
            (LocalVariables|SetLocalBool true)))
        (:Completed
          _)))
    (:NewEnumerator1
      _))
  (return (LocalVariables|GetTargetActorResult) (LocalVariables|GetLocalBool)))

(fn FadeInAmbientLoop
  (Audio|Components|Audio|FadeIn (Variables|Default|GetAmbient%20Loop) 0.1 1.0 0.0 "Linear" :!class /Script/Engine.AudioComponent))

(fn FadeOutAmbientLoop
  (Audio|Components|Audio|FadeOut (Variables|Default|GetAmbient%20Loop) 0.5 0.0 "Linear" :!class /Script/Engine.AudioComponent)
  (SoundIndicator|StopSoundIndicatorLooping (Variables|Default|GetFortSoundIndicator)))

(fn IsPlantBetweenInstigatorAndTarget
  (params (InstigatingActor :type Actor%20Object%20Reference) (Target :type Actor%20Object%20Reference))
  (return (> (Math|Vector|DotProduct (Transformation|GetActorForwardVector InstigatingActor) (Math|Vector|Normalize (- (Transformation|GetActorLocation Target) (Transformation|GetActorLocation)))) 0)))

(fn MakeExplosionContainerData
  (bind _returnvalue (Ability|MakeGameplayEffectContainerSpecfromAbilityComponent (Variables|ChildOverrides|GetEffectContainerExplode) (select (Utilities|IsValid (Ability|GetAbilitySystemComponent (Variables|Default|GetInstigatingActor)) :!class /Script/Engine.KismetSystemLibrary) (Ability|GetAbilitySystemComponent (Variables|Default|GetInstigatingActor)) (Ability|GetAbilitySystemComponent self))))
  (return _returnvalue))

(fn MakeHitContainerData
  (bind _returnvalue (Ability|MakeGameplayEffectContainerSpecfromAbilityComponent (Variables|ChildOverrides|GetEffectContainerHit) (select (Utilities|IsValid (Ability|GetAbilitySystemComponent (Variables|Default|GetInstigatingActor)) :!class /Script/Engine.KismetSystemLibrary) (Ability|GetAbilitySystemComponent (Variables|Default|GetInstigatingActor)) (Ability|GetAbilitySystemComponent self))))
  (return _returnvalue))

(fn MakeProjectile
  (bind (_hitnormal_x _hitnormal_y) (Math|Vector|BreakVector (Variables|GetHitNormal_X)))
  ;[0.4895 0.1317 1] Create projectile and set its effects / static mesh
  (bind _returnvalue (CallFunction|MakeExplosionContainerData))
  (bind _returnvalue_2 (CallFunction|MakeHitContainerData))
  (bind (_targetactor _foundtarget) (CallFunction|DoTraceforInfluence))
  (LocalVariables|SetLocalTargetActor _targetactor)
  (bind _returnvalue_3 (Projectile|SpawnProjectilewithParams (Variables|Default|GetProjectileClass) (Variables|Default|GetInstigatingActor) (Utilities|Struct|MakeSpawnProjectileParams (+ (Transformation|GetWorldLocation (Variables|Default|GetLaunchableMesh)) (LocalVariables|GetSpawnOffset)) "0.000000,0.000000,0.000000" (Math|Conversions|RotationFromXVector (Math|Vector|MakeVector _hitnormal_x _hitnormal_y 0.6)) _returnvalue_2 _returnvalue false 0 (ScalableFloat|GetValueatLevel (Variables|ChildOverrides|GetLaunchSpeed)) (Variables|Projectile|GetProjectileMaxSpeed))))
  (if _foundtarget
    (CallFunction|SetCustomPlantVelocity _returnvalue_3 (LocalVariables|GetLocalTargetActor))
    (else
      _))
  (bind _asprj_chrono_launchable_plant_parent (Utilities|Casting|CastToPrj_Chrono_LaunchablePlant_Parent _returnvalue_3))
  (Class|PrjChronoLaunchablePlantParent|SetProjectileMesh _asprj_chrono_launchable_plant_parent (Variables|/Script/Engine.StaticMeshComponent|GetStaticMesh (Variables|Default|GetLaunchableMesh)) :!class /ChronoEnviroItems/LaunchablePlants/Gameplay/Actors/Prj_Chrono_LaunchablePlant_Parent.Prj_Chrono_LaunchablePlant_Parent_C)
  (Actor|SetOwner _asprj_chrono_launchable_plant_parent (Variables|Default|GetInstigatingActor))
  (Components|Activation|Activate (Variables|/Script/FortniteGame.FortProjectileBase|GetProjectileMovementComponent _asprj_chrono_launchable_plant_parent) false)
  (return _asprj_chrono_launchable_plant_parent))

(fn OnProjectileHit
  (params (Hit :type Hit%20Result%20Structure))
  (Development|PrintString "plant pod hit")
  (Rendering|Material|SetScalarParameterValue (Variables|Default|GetBaseMID) "GameTime_Explosion_Base" (Utilities|Time|GetGameTimeinSeconds) :!class /Script/Engine.MaterialInstanceDynamic)
  (CallFunction|FadeOutAmbientLoop))

(fn OnRep_bWaitingToRegrow
  (if (Variables|Default|GetbWaitingToRegrow)
    (Transformation|SetWorldScale3D (Variables|Default|GetLaunchableMesh) "0, 0, 0")
    (CallFunction|FadeOutAmbientLoop)
    (else
      (Utilities|FlowControl|Sequence
        (:then_0
          (CallFunction|FadeinAmbientLoop))
        (:then_1
          (Transformation|SetWorldScale3D (Variables|Default|GetLaunchableMesh) "1.000000,1.000000,1.000000")
          (Rendering|Material|SetScalarParameterValueonMaterials (Variables|Default|GetLaunchableMesh) "GameTime" (Utilities|Time|GetGameTimeinSeconds)))
        (:then_2
          (Collision|SetCollisionEnabled (Variables|Default|GetLaunchableMesh) "QueryAndPhysics")
          (Ability|GameplayEffect|RemoveLooseGameplayTags self (Variables|Default|GetRegrowthTags))
          (Audio|PlaySoundatLocation (Variables|Default|GetRegrowth_Sound) (Transformation|GetActorLocation)))))))

(fn PerformOverlapActorCheck
  (bind (_outhit _returnvalue) (Collision|BoxTraceForObjects :Start (Transformation|GetActorLocation) :End (Transformation|GetActorLocation) :HalfSize "128.000000,128.000000,128.000000" :Orientation "0, 0, 0" :ObjectTypes (Utilities|Array|MakeArray "ObjectTypeQuery3") :bTraceComplex false :DrawDebugType "ForDuration" :bIgnoreSelf true :TraceColor "(R=1.000000,G=0.000000,B=0.000000,A=1.000000)" :TraceHitColor "(R=0.000000,G=1.000000,B=0.000000,A=1.000000)" :DrawTime 5.0))
  (if _returnvalue
    ;[0.2144 0.6667 0.07986] We keep attempting to regrow each time this pass is false
    (Utilities|FlowControl|Sequence
      (:then_0
        (Utilities|Time|SetTimerbyEvent (EventDispatchers|CreateEvent :!event TryRegrowth) (Variables|Default|GetCheckRegrowthInterval)))
      (:then_1
        (return false)))
    (else
      (return true))))

(fn ResetHealth
  (Health|SetHealth (Health|GetMaxHealth))
  (return))

(fn SelectVectorByTag
  (params (Tags :type Gameplay%20Tag%20Container%20Structure) (HitNormal :type Vector))
  ;[0.8089 0.1457 1] Use player's forward vector unless the source is explosive
  (return (Math|Vector|SelectVector (Math|Vector|NegateVector HitNormal) (Transformation|GetActorForwardVector (Variables|Default|GetInstigatingActor)) (GameplayTags|HasTag Tags "(TagName=\"Gameplay.Damage.Physical.Explosive\")"))))

(fn SetCustomPlantVelocity
  (params (Projectile :type Fort%20Projectile%20Base%20Object%20Reference) (TargetActor :type Actor%20Object%20Reference))
  (bind _returnvalue (Transformation|GetActorLocation))
  ;[0.4895 0.1317 1] Set Projectile velocity towards influential target
  (bind _outlaunchvelocity (Game|SuggestProjectileVelocityMovingTarget :ProjectileStartLocation _returnvalue :TargetActor TargetActor :GravityZOverride 0.0 :TimeToTarget (Variables|Projectile|GetProjectileTimeToTarget) :DrawDebugType "None" :DrawDebugTime 3.0 :DrawDebugColor "(R=1.000000,G=0.000000,B=0.000000,A=1.000000)"))
  (Transformation|SetActorRotation Projectile (Math|Rotator|FindLookatRotation _returnvalue (Transformation|GetActorLocation TargetActor)) false)
  (Projectile|SetVelocity (Variables|/Script/FortniteGame.FortProjectileBase|GetProjectileMovementComponent Projectile) _outlaunchvelocity)
  (return))

(fn SetInstigatingActorWithCheck
  (params (Actor :type Actor%20Object%20Reference))
  (bind _asfort_player_state_athena (Utilities|Casting|CastToFortPlayerStateAthena Actor)
    (:then
      (Utilities|IsValid (PlayerState|GetCurrentPawn _asfort_player_state_athena)
        (:Is%20Valid
          (Variables|Default|SetInstigatingActor (PlayerState|GetCurrentPawn _asfort_player_state_athena)))
        (:Is%20Not%20Valid
          (Variables|Default|SetInstigatingActor Actor)))
      (return))
    (:CastFailed
      _)))

(fn ConstructionScript
  (bind _returnvalue (Rendering|Material|CreateDynamicMaterialInstance (Variables|Default|GetBaseMesh) 0 0 "None" :!class /Script/Engine.PrimitiveComponent))
  (bind _output_get (Variables|Default|SetBaseMID _returnvalue))
  (Rendering|Material|SetScalarParameterValue _output_get "PlantTendrilType" (Variables|Default|GetPlantTendrilType) :!class /Script/Engine.MaterialInstanceDynamic))
