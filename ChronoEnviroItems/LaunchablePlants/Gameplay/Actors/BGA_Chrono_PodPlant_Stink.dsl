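; BGA_Chrono_PodPlant_Stink is the stink variant of a launchable pod plant gameplay actor.
; Its construction script runs the parent's construction logic, then sets the PlantTendrilType scalar parameter to zero on the base mesh materials.
; The file holds no other logic of its own, only this visual tendril selection for the variant.
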(fn ConstructionScript
  (|CallParentFunction)
  (Rendering|Material|SetScalarParameterValueonMaterials (Variables|Default|GetBaseMesh) "PlantTendrilType" 0.0))
