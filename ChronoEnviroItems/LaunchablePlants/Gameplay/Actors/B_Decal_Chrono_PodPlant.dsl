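; Decal actor for the Chrono pod plant that animates its decal material in when it begins play.
; On BeginPlay it creates a dynamic instance of the decal's material, stores it as Decal_MID and assigns it back to the decal component.
; A timeline then plays from the start, driving the material's FadeIn and Opacity scalar parameters on each update.
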(event EventBeginPlay
  (bind _returnvalue (Rendering|Material|CreateDynamicMaterialInstance (Rendering|Components|Decal|GetDecalMaterial (Variables|Decal|GetDecal)) "None" "None" :!class /Script/Engine.KismetMaterialLibrary))
  (bind _output_get (Variables|Default|SetDecal_MID _returnvalue))
  (Rendering|Components|Decal|SetDecalMaterial (Variables|Decal|GetDecal) _output_get)
  (bind (_direction _fadein _opacity) (Timeline|Timeline PlayFromStart)
    (:Update
      (Rendering|Material|SetScalarParameterValue (Variables|Default|GetDecal_MID) "FadeIn" _fadein :!class /Script/Engine.MaterialInstanceDynamic)
      (Rendering|Material|SetScalarParameterValue (Variables|Default|GetDecal_MID) "Opacity" _opacity :!class /Script/Engine.MaterialInstanceDynamic))))

(fn ConstructionScript)
