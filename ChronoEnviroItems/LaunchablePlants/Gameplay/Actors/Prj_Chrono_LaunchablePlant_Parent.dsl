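; Parent projectile for Chrono launchable plant pods, which bounce until they explode and leave a decal.
; The server triggers explosions on a timer, excess bounces or pawn/building hits; timed explosions trace downward for the decal normal.
; BlowUp stops movement and animates explosion materials before detonating, optionally delayed.
; It also fits the capsule to mesh bounds, drives water Niagara effects, and broadcasts explosions, adding the instigator's controller in Creative playlists.
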(event OnEnterWaterEvent(WaterInteractionComponent)
  (params (WaterBody :type Fort%20Water%20Body%20Actor%20Object%20Reference) (WaterInteractionComponent :type Fort%20Water%20Interaction%20Component%20Object%20Reference) (bIsFirstBody :type Boolean))
  (bind _returnvalue (Transformation|GetActorLocation))
  ;[0.2144 0.6667 0.07986] Play niagara when entering water
  (bind _outwatersurfaceinfo (WaterBody|GetWaterSurfaceInfoStruct WaterBody _returnvalue true false false))
  (Niagara|SpawnSystematLocation (Variables|Default|GetNS_PrjEnterWater) _returnvalue)
  (bind _returnvalue_2 (Niagara|SpawnSystemAttached (Variables|Default|GetNS_PrjFloatingInWater) (Variables|Default|GetProjectileMesh)))
  (bind _output_get (Variables|Default|SetFloatingNSComp _returnvalue_2))
  (Effects|Components|ParticleSystem|SetFloatParameter _output_get "User.WaterDepth" (.WaterDepth (Utilities|Struct|BreakWaterSurfaceInfo _outwatersurfaceinfo)))
  (Effects|Components|ParticleSystem|SetVectorParameter _output_get "User.WaterPlaneLocation" (Utilities|Struct|BreakWaterSurfaceInfo _outwatersurfaceinfo))
  (Effects|Components|ParticleSystem|SetVectorParameter _output_get "User.WaterPlaneNormal" (.WaterPlaneNormal (Utilities|Struct|BreakWaterSurfaceInfo _outwatersurfaceinfo))))

(event EventDestroyed
  (bind _floatingnscomp (Variables|Default|GetFloatingNSComp))
  ;[0.2144 0.6667 0.07986] Destroy lingering component if valid
  (Utilities|IsValid _floatingnscomp
    (:Is%20Valid
      (Components|DestroyComponent _floatingnscomp))))

(event EventOnExploded
  (params (HitActors :type Array%20of%20Actor%20Object%20References) (HitResults :type Array%20of%20Hit%20Result%20Structures))
  (bind _floatingnscomp (Variables|Default|GetFloatingNSComp))
  (|CallParentFunction HitActors HitResults)
  (Default|CallOnPlantExploded (Variables|Default|GetPodHit))
  ;[0.8528 1 0.3833 0.5] For the associated Creative devices, send the projectile's instigator
  (if (Math|Boolean|ORBoolean (Class|FortGameStateAthena|IsAnyCreativePlaylist (Fortnite|Athena|GameState|GetGameStateAthena)) (Class|FortGameStateAthena|IsCreativeRoyalePlaylist (Fortnite|Athena|GameState|GetGameStateAthena)) (Class|FortGameStateAthena|IsVKPlayPlaylist (Fortnite|Athena|GameState|GetGameStateAthena)))
    (Default|CallOnPlantExplodedSendInstigator (Fortnite|GetControllerfromActor (Variables|Actor|GetInstigator)))
    (else
      _))
  (Audio|Components|Audio|FadeOut (Variables|Default|GetInAirLoop) 0.25 0.0 "Linear" :!class /Script/Engine.AudioComponent) ; Fade out In air Loop 
  (CallFunction|SpawnDecal)
  (Utilities|FlowControl|Sequence
    (:then_0
      ;[0.2144 0.6667 0.07986] Play additional niagara when exploding in water
      (if (Water|IsinWater (Variables|Water|GetWaterInteractionComponent))
        (Niagara|SpawnSystematLocation (Variables|Default|GetNS_PrjExplodeInWater) (Transformation|GetActorLocation))))
    (:then_1
      ;;[0.2144 0.6667 0.07986] Turn off the floating NS when exploding (if valid)
      (Utilities|IsValid _floatingnscomp
        (:Is%20Valid
          (Components|Activation|Deactivate _floatingnscomp))))))

(event EventBeginPlay
  (Networking|SwitchHasAuthority
    (:Authority
      (bind _returnvalue (Utilities|Time|SetTimerbyEvent (delegate Custom|Explode) (ScalableFloat|GetValueatLevel (Variables|Default|GetTimeToExplode))))
      (Variables|Default|SetTimeToExplodeTimerHandle _returnvalue))))

(event Custom|Explode
  (Utilities|Time|ClearandInvalidateTimerbyHandle (Variables|Default|GetTimeToExplodeTimerHandle))
  (CallFunction|MakeStaticNormal)
  (Delivery|DoExplosion))

(event Custom|SetProjectileMesh
  (params (Mesh :type Static%20Mesh%20Object%20Reference))
  (CallFunction|UpdateMeshandCollider Mesh))

(event Custom|BlowUp
  (params (bExplodeInstantly :type Boolean) (OverridePlantBase :type Static%20Mesh%20Component%20Object%20Reference))
  (bind _returnvalue (Utilities|Time|GetGameTimeinSeconds))
  (bind _projectilemesh (Variables|Default|GetProjectileMesh))
  (bind _overridebasemesh (Variables|Default|GetOverrideBaseMesh))
  (bind _sourceactor (Variables|Default|GetSourceActor))
  (Variables|Default|SetbExplodingInstantly bExplodeInstantly)
  (Utilities|IsValid OverridePlantBase
    (:Is%20Valid
      (Variables|Default|SetOverrideBaseMesh OverridePlantBase))
    (:Is%20Not%20Valid
      _)) ; Used for the device version of the plants
  (Components|Activation|Deactivate (Variables|Movement|GetProjectileMovementComponent))
  (Variables|/Script/Engine.RotatingMovementComponent|SetRotationRate "0, 0, 0" (Variables|Default|GetRotatingMovement))
  (Utilities|FlowControl|DoOnce false
    (:Completed
      (Utilities|FlowControl|Sequence
        (:then_0
          (Rendering|Material|SetScalarParameterValueonMaterials _projectilemesh "WPOPuslingScale" 0.4)
          (Rendering|Material|SetScalarParameterValueonMaterials _projectilemesh "GameTime_Explosion" (Utilities|Time|GetGameTimeinSeconds))
          (if (not (Variables|Default|GetbExplodingInstantly))
            (Utilities|FlowControl|Delay 0.5))
          (Audio|Components|Audio|Stop (stale _returnvalue_2))
          (CallFunction|Explode))
        (:then_1
          (bind _returnvalue_2 (Audio|SpawnSoundatLocation :Sound "/ChronoEnviroItems/LaunchablePlants/Gameplay/Assets/Sounds/zz_Proto/PodPlant_Proto_GrowthCue.PodPlant_Proto_GrowthCue" :Location (Transformation|GetWorldLocation (Variables|Collision|GetCapsuleComponent)) :VolumeMultiplier 4.0 :PitchMultiplier 1.0 :StartTime 0.0 :AttenuationSettings "/Game/Sounds/Attenuation/Fort_Ability_Radius.Fort_Ability_Radius" :bAutoDestroy true)))
        (:then_2
          (Utilities|IsValid _overridebasemesh
            (:Is%20Valid
              (Rendering|Material|SetScalarParameterValueonMaterials _overridebasemesh "GameTime_Explosion_Base" _returnvalue))
            (:Is%20Not%20Valid
              (Utilities|IsValid _sourceactor
                (:Is%20Valid
                  (Rendering|Material|SetScalarParameterValue (Variables|/ChronoEnviroItems/LaunchablePlants/Gameplay/Actors/BGA_Chrono_LaunchablePlant_Parent.BGA_Chrono_LaunchablePlant_Parent_C|GetBaseMID _sourceactor) "GameTime_Explosion_Base" _returnvalue :!class /Script/Engine.MaterialInstanceDynamic))))))))))

(event OnProjectileBounce(ProjectileMovementComponent)
  (params (ImpactResult :type Hit%20Result%20Structure%20%28by%20ref%29) (ImpactVelocity :type Vector%20%28by%20ref%29 :default "0, 0, 0"))
  ;[0.7054 0.3213 1] Bounce
  (Networking|SwitchHasAuthority
    (:Authority
      (bind _result (Math|Integer|IncrementInt (Variables|Default|GetCurrentBounces))
        (:%20%20
          (if (> _result (Variables|Default|GetMaxBounces))
            (Variables|Default|SetPodHit ImpactResult)
            (Delivery|DoExplosion)
            (Utilities|Time|ClearandInvalidateTimerbyHandle (Variables|Default|GetTimeToExplodeTimerHandle))))))))

(event Collision|EventHit
  (params (MyComp :type Primitive%20Component%20Object%20Reference) (Other :type Actor%20Object%20Reference) (OtherComp :type Primitive%20Component%20Object%20Reference) (bSelfMoved :type Boolean) (HitLocation :type Vector :default "0, 0, 0") (HitNormal :type Vector :default "0, 0, 0") (NormalImpulse :type Vector :default "0, 0, 0") (Hit :type Hit%20Result%20Structure%20%28by%20ref%29))
  ;[0.03429 0.3069 0.7899] Explode if hit pawn
  (Networking|SwitchHasAuthority
    (:Authority
      (bind _returnvalue (GameplayTags|HasAnyMatchingGameplayTags(Message) Other (GameplayTags|MakeLiteralGameplayTagContainer "(GameplayTags=((TagName=\"Pawn.Athena\"),(TagName=\"Building\")))")))
      (if _returnvalue
        (Variables|/Script/Engine.ProjectileMovementComponent|SetbShouldBounce false (Variables|Movement|GetProjectileMovementComponent))
        (Variables|Default|SetPodHit Hit)
        (Utilities|Time|ClearandInvalidateTimerbyHandle (Variables|Default|GetTimeToExplodeTimerHandle))
        (Delivery|DoExplosion)))))

(fn MakeStaticNormal
  (bind _outhit (Collision|LineTraceByChannel (+ (Transformation|GetActorLocation) "0.000000,0.000000,50.000000") (- (Transformation|GetActorLocation) "0.000000,0.000000,512.000000") "TraceTypeQuery10" false (Utilities|Array|MakeArray self)))
  (Variables|Default|SetPodHit _outhit)
  (return))

(fn SpawnDecal
  (bind _podhit (Variables|Default|GetPodHit))
  (bind (_bblockinghit _binitialoverlap _time _distance _location _impactpoint _normal) (Collision|BreakHitResult _podhit))
  ;[0 0.8244 1] Decal Spawn
  (bind _returnvalue (Game|SpawnActorfromClass (Variables|Default|GetDecalClass) (Math|Transform|MakeTransform _location) "AlwaysSpawn" "OverrideRootScale" 0 0))
  (Transformation|SetWorldRotation (Variables|/Script/Engine.Actor|GetRootComponent _returnvalue) (Math|Rotator|MakeRotator (+ (.roll (Math|Rotator|MakeRotfromX _normal)) (Math|Random|RandomFloatinRange 0.0 3.6e+02)) (* (.pitch (Math|Rotator|MakeRotfromX _normal)) -1.0) (.yaw (Math|Rotator|MakeRotfromX _normal))) false false))

(fn UpdateMeshAndCollider
  (params (Mesh :type Static%20Mesh%20Object%20Reference))
  (bind _projectilemesh (Variables|Default|GetProjectileMesh))
  (bind _staticmesh (Variables|/Script/Engine.StaticMeshComponent|GetStaticMesh _projectilemesh))
  (bind (_returnvalue_min _returnvalue_max) (Utilities|Struct|BreakBox (StaticMesh|GetBoundingBox _staticmesh)))
  (bind (_returnvalue_min_x _returnvalue_min_y _returnvalue_min_z) (Math|Vector|BreakVector _returnvalue_min))
  (bind (_returnvalue_max_x _returnvalue_max_y _returnvalue_max_z) (Math|Vector|BreakVector _returnvalue_max))
  (bind _collisionsizeadjustment (LocalVariables|GetCollisionSizeAdjustment))
  (bind _newhalfheight (LocalVariables|GetNewHalfHeight))
  (bind _capsulecomponent (Variables|Collision|GetCapsuleComponent))
  (Components|StaticMesh|SetStaticMesh (Variables|Default|GetProjectileMesh) Mesh)
  ;[1 0.9411 0.3122 0.5] Calculate new collider dimensions
  (LocalVariables|SetNewHalfHeight (- (/ (- _returnvalue_max_z _returnvalue_min_z) 2.0) _collisionsizeadjustment))
  (if (Variables|Default|GetbForceSphereCollision)
    (LocalVariables|SetNewRadius (LocalVariables|GetNewHalfHeight))
    (else
      (LocalVariables|SetNewRadius (- (/ (Math|Float|Max(Float) (- _returnvalue_max_x _returnvalue_min_x) (- _returnvalue_max_y _returnvalue_min_y)) 2.0) _collisionsizeadjustment))))
  (Components|Capsule|SetCapsuleRadius _capsulecomponent (LocalVariables|GetNewRadius) false)
  (Components|Capsule|SetCapsuleHalfHeight _capsulecomponent _newhalfheight false)
  (Transformation|SetRelativeLocation (Variables|Default|GetProjectileMesh) (Math|Vector|MakeVector 0.0 0.0 (* _newhalfheight -1.0)) false false) ; These projectile meshes all have their pivots at the bottom.\nThe relative location must take that into account.
)

(fn ConstructionScript)
