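; Projectile Blueprint for the bomb variant of the Chrono pod plant, handling audio-visualizer feedback when it explodes.
; On explosion it runs the parent logic, then shows a local gunshot-type sound indicator at its location, with audible range read from a scalable float.
; It then stops and deactivates its looping FortSoundIndicator component.
; The construction script only calls the parent.
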(event EventOnExploded
  (params (HitActors :type Array%20of%20Actor%20Object%20References) (HitResults :type Array%20of%20Hit%20Result%20Structures))
  (bind _fortsoundindicator (Variables|Default|GetFortSoundIndicator))
  (|CallParentFunction HitActors HitResults)
  (Fortnite|ShowSoundIndicatorLocal :Location (Transformation|GetActorLocation) :MaxAudibleDistance (ScalableFloat|GetValueatLevel (Variables|Default|GetHF_MaxAVDistance)) :IndicatorType "Gunshot" :Instigator (Variables|Actor|GetInstigator) :Tint "(R=0.000000,G=0.000000,B=0.000000,A=0.000000)" :AttenuationMaxRangeMultiplier 1.0 :AttenuationIntensityMultiplier 1.0)
  (SoundIndicator|StopSoundIndicatorLooping _fortsoundindicator)
  (Components|Activation|Deactivate _fortsoundindicator))

(fn ConstructionScript
  (|CallParentFunction))
