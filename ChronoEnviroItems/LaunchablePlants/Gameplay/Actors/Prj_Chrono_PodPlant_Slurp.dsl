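; Slurp variant of the launchable pod plant projectile, extending its explosion so slurp also reaches vehicle passengers.
; On explosion it casts each hit actor to the Fort vehicle interface and gathers all passengers, then calls the parent explosion handler.
; A helper applies each gameplay effect spec from the on-explode effect container to the passenger's ability system component, with the projectile as source object.
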(event EventOnExploded
  (params (HitActors :type Array%20of%20Actor%20Object%20References) (HitResults :type Array%20of%20Hit%20Result%20Structures))
  ;[0.0356 0.414 0.6667] Iterate through vehicle occupants so we can apply slurp to passengers
  (for _array_element HitActors
    (bind _asfort_vehicle_interface (Utilities|Casting|CastToFortVehicleInterface _array_element))
    (for _array_element_2 (Vehicle|GetAllPassengers _asfort_vehicle_interface)
      (CallFunction|ApplySpectoPassenger _array_element_2)))
  (|CallParentFunction HitActors HitResults))

(fn ApplySpecToPassenger
  (params (Player :type Fort%20Player%20Pawn%20Object%20Reference))
  (for _array_element (Utilities|Struct|BreakFortGameplayEffectContainerSpec (Variables|Delivery|GetEffectContainerSpecToApplyOnExplode))
    (GameplayEffects|FortApplyGameplayEffectSpecToTarget :self (Ability|GetFortAbilitySystemComponent Player) :GameplayEffectSpecHandle _array_element :TargetOfEffect (Ability|GetFortAbilitySystemComponent Player) :OptionalSourceObject self))
  (return))

(fn ConstructionScript)
