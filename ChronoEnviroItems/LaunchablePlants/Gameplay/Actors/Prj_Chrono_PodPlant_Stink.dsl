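; Stink pod plant projectile that leaves a stink impact cloud where it explodes.
; On explosion it runs the parent handler, then spawns the impact cloud building gameplay actor at its location with its instigator passed through.
; The spawned cloud is initialized with the instigator's player controller.
; It also stops the looping sound indicator and deactivates that component.
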(event EventOnExploded
  (params (HitActors :type Array%20of%20Actor%20Object%20References) (HitResults :type Array%20of%20Hit%20Result%20Structures))
  (bind _returnvalue (Game|GetInstigator))
  (bind _fortsoundindicator (Variables|Default|GetFortSoundIndicator))
  (|CallParentFunction HitActors HitResults)
  (bind _returnvalue_2 (Building|SpawnBuildingGameplayActor "/ChronoEnviroItems/LaunchablePlants/Gameplay/Actors/BGA_Chrono_PodPlant_Stink_ImpactCloud.BGA_Chrono_PodPlant_Stink_ImpactCloud_C" (Math|Transform|MakeTransform (Transformation|GetActorLocation)) _returnvalue))
  (Building|InitializeKismetSpawnedBuildingActor _returnvalue_2 0 (Fortnite|GetFortPlayerControllerfromActor _returnvalue) true 0 false)
  (SoundIndicator|StopSoundIndicatorLooping _fortsoundindicator)
  (Components|Activation|Deactivate _fortsoundindicator))

(fn ConstructionScript)
